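# Sound cue for shrub impacts in the demolition impact sound set, attenuated as a small-object hit in the impacts sound class.
# A quality-level node selects among four branches that step down from three random wave variations to two, two, then a single wave.
# Each branch passes through a modulator for slight pitch variation.
# A four-max object-impact concurrency set is assigned, alongside a prevent-new concurrency override.
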SoundCue /Game/Sounds/Fort_Impact_Sounds/Demolition/Bush/Demo_Shrub_Impact_Cue
root=QualityLevel_0
nodes [
 {id=Modulator_0, children [Random_0], properties {pitchMax=1.07, pitchMin=0.92}}
 {id=Modulator_2, children [Random_1], properties {pitchMax=1.07, pitchMin=0.92}}
 {id=Modulator_3, children [Random_2], properties {pitchMax=1.07, pitchMin=0.92}}
 {id=Modulator_5, children [WavePlayer_7], properties {pitchMax=1.1, pitchMin=0.9}}
 {id=QualityLevel_0, children [Modulator_0,Modulator_2,Modulator_3,Modulator_5]}
 {id=Random_0, children [WavePlayer_0,WavePlayer_1,WavePlayer_2], properties {weights [1,1,1]}}
 {id=Random_1, children [WavePlayer_3,WavePlayer_4], properties {weights [1,1]}}
 {id=Random_2, children [WavePlayer_5,WavePlayer_6], properties {weights [1,1]}}
 {id=WavePlayer_0, properties {soundWaveAssetPtr=&/shrub_impact_01}}
 {id=WavePlayer_1, properties {soundWaveAssetPtr=&/shrub_impact_02}}
 {id=WavePlayer_2, properties {soundWaveAssetPtr=&/shrub_impact_03}}
 {id=WavePlayer_3, properties {soundWaveAssetPtr=&shrub_impact_01}}
 {id=WavePlayer_4, properties {soundWaveAssetPtr=&shrub_impact_02}}
 {id=WavePlayer_5, properties {soundWaveAssetPtr=&shrub_impact_01}}
 {id=WavePlayer_6, properties {soundWaveAssetPtr=&shrub_impact_02}}
 {id=WavePlayer_7, properties {soundWaveAssetPtr=&shrub_impact_01}}
]
properties {
 attenuationSettings=/Game/Sounds/Attenuation/Fort_Demolition_Hit_Small_Sized_Object_Att
 maxDistance=4000
 soundClassObject=&Fort_Impacts
 volumeMultiplier=0.68
 concurrencyOverrides {resolutionRule=PreventNew}
 concurrencySet [&CONC_Object_Impacts_FourMax]
}
